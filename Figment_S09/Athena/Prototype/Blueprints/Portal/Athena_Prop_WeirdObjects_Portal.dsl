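; Weird Objects portal prop that mounts on a wall and teleports to a linked partner portal.
; On BeginPlay, authority line traces find the wall and bottom piece, and the child BGA_Portal actor, which does the teleporting, is wired to the partner.
; Teleports trigger a timeline pulsing the backer material and point light.
; Wall death destroys the prop; its server death event destroys the partner, bottom piece and child portal component.
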(event EventBeginPlay
  (bind _woportal_actorref (Variables|Default|GetWOPortal_ActorRef))
  (|CallParentFunction)
  (Utilities|FlowControl|Sequence
    (:then_0
      ;[1 1 1] Line trace to get attached wall
      (Networking|SwitchHasAuthority
        (:Authority
          (bind (_outhit _returnvalue) (Collision|LineTraceByChannel (+ (Transformation|GetActorLocation) "0, 0,50.000000") (+ (+ (Transformation|GetActorLocation) "0, 0,50.000000") (* (Transformation|GetActorForwardVector) -5e+01)) "TraceTypeQuery10" false (Utilities|Array|MakeArray self)))
          (if _returnvalue
            (bind _asbuilding_actor (Utilities|Casting|CastToBuildingActor (.HitActor (Collision|BreakHitResult _outhit))))
            (bind _output_get (Variables|Default|SetWall _asbuilding_actor))
            (EventDispatchers|BindEventtoOnDamaged _output_get (EventDispatchers|CreateEvent :!event Custom|OnWallDamaged))
            (EventDispatchers|BindEventtoOnDied (Variables|Default|GetWall) (EventDispatchers|CreateEvent :!event Custom|OnWallDied))))))
    (:then_1
      ;[1 1 1] Get BottomPiece with line trace
      (Networking|SwitchHasAuthority
        (:Authority
          (bind (_outhit_2 _returnvalue_2) (Collision|LineTraceByChannel :Start (Math|Vector|vector+vector (Transformation|GetActorLocation) "0, 0,75.000000" (* (Transformation|GetActorRightVector) 75.0)) :End (+ (Math|Vector|vector+vector (Transformation|GetActorLocation) "0, 0,75.000000" (* (Transformation|GetActorRightVector) 75.0)) "0, 0,-100.000000") :TraceChannel "TraceTypeQuery10" :bTraceComplex false :DrawDebugType "Persistent" :bIgnoreSelf true :TraceColor "(R=1.000000,G=0.000000,B=0.000000,A=1.000000)" :TraceHitColor "(R=0.000000,G=1.000000,B=0.000000,A=1.000000)" :DrawTime 5e+02))
          (if _returnvalue_2
            (Variables|Default|SetBottomPiece (.HitActor (Collision|BreakHitResult _outhit_2)))))))
    (:then_2
      ;[1 1 1] Set Actor Ref of child actor (BGA_Portal, the part that does the actual teleportation)
      (bind _asbga_portal_weird_objects_athena (Utilities|Casting|CastToBGA_Portal_WeirdObjects_Athena (Variables|/Script/Engine.ChildActorComponent|GetChildActor (Variables|Default|GetBGA_Portal_WeirdObjects_Athena))))
      (Default|BindEventtoTeleport _asbga_portal_weird_objects_athena (EventDispatchers|CreateEvent :!event Custom|TeleportCalled))
      (Variables|/Figment_S09/Athena/Prototype/Blueprints/Portal/BGA_Portal_WeirdObjects_Athena.BGA_Portal_WeirdObjects_Athena_C|SetFrom%20Portal self _asbga_portal_weird_objects_athena)
      (Variables|/Figment_S09/Athena/Prototype/Blueprints/Portal/BGA_Portal_WeirdObjects_Athena.BGA_Portal_WeirdObjects_Athena_C|SetTo%20Portal _woportal_actorref _asbga_portal_weird_objects_athena)
      (for _array_element (Actor|GetAllChildActors _woportal_actorref false)
        (bind _asbga_portal_weird_objects_athena_2 (Utilities|Casting|CastToBGA_Portal_WeirdObjects_Athena _array_element))
        (Variables|/Game/Athena/Prototype/Blueprints/Portal/BGA_Portal_Athena.BGA_Portal_Athena_C|SetActorRef _asbga_portal_weird_objects_athena_2 _asbga_portal_weird_objects_athena)))))

(event Custom|OnWallDamaged
  (params (DamagedActor :type Actor%20Object%20Reference) (Damage :type Float%20%28single-precision%29) (InstigatedBy :type Controller%20Object%20Reference) (DamageCauser :type Actor%20Object%20Reference) (HitLocation :type Vector) (FHitComponent :type Primitive%20Component%20Object%20Reference) (BoneName :type Name) (Momentum :type Vector))
  ;[1 1 1] Destroy self if wall dies.
  (bind _asbuilding_actor (Utilities|Casting|CastToBuildingActor DamagedActor))
  (if (<= (Health|GetHealth _asbuilding_actor) 0.0)
    (Utilities|FlowControl|DoOnce false
      (:Completed
        (Actor|DestroyActor)))))

(event Custom|OnWallDied
  (params (DamagedActor :type Actor%20Object%20Reference) (Damage :type Float%20%28single-precision%29) (InstigatedBy :type Controller%20Object%20Reference) (DamageCauser :type Actor%20Object%20Reference) (HitLocation :type Vector) (FHitComponent :type Primitive%20Component%20Object%20Reference) (BoneName :type Name) (Momentum :type Vector))
  ;[1 1 1] Destroy self if wall dies.
  (Utilities|FlowControl|DoOnce false
    (:Completed
      (Actor|DestroyActor))))

(event Custom|TeleportCalled
  ;[0.1887 0.7885 1] Teleport Event
  (if (not (Components|Timeline|IsPlaying (Variables|Athena_Prop_WeirdObjects_Portal|GetPortalPulse)))
    (bind (_direction _portal _light) (Timeline|PortalPulse PlayFromStart)
      (:Update
        (Rendering|Material|SetScalarParameterValueonMaterials (Variables|Default|GetS_FX_WeirdObjects_Backer) "Active" _portal)
        (Rendering|Components|Light|SetIntensity (Variables|Default|GetPortal%20Point) (Math|Float|Lerp 1e+01 2e+01 _light))))))

(event EventOnDeathServer
  (params (Damage :type Float%20%28single-precision%29) (DamageTags :type Gameplay%20Tag%20Container%20Structure%20%28by%20ref%29) (Momentum :type Vector :default "0, 0, 0") (HitInfo :type Hit%20Result%20Structure%20%28by%20ref%29) (InstigatedBy :type Controller%20Object%20Reference) (DamageCauser :type Actor%20Object%20Reference) (EffectContext :type Gameplay%20Effect%20Context%20Handle%20Structure))
  (bind _bottompiece (Variables|Default|GetBottomPiece))
  (bind _woportal_actorref (Variables|Default|GetWOPortal_ActorRef))
  ;[1 1 1] On death, destroy connected portal, bottom piece, and child actor (BGA_Portal).
  (Utilities|IsValid _bottompiece
    (:Is%20Valid
      (Actor|DestroyActor _bottompiece))
    (:Is%20Not%20Valid
      _))
  (Utilities|IsValid _woportal_actorref
    (:Is%20Valid
      (Actor|DestroyActor _woportal_actorref))
    (:Is%20Not%20Valid
      _))
  (Components|DestroyComponent (Variables|Default|GetBGA_Portal_WeirdObjects_Athena)))

(fn ConstructionScript)
