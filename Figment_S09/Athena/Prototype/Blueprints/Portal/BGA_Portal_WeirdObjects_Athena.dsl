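; Portal actor Blueprint that plays the effects when a player pawn teleports between its From and To portals.
; After a short delay, EventPlayTeleportFX spawns the enter-portal emitter at both openings, using each portal's offset and rotation, and fires a local burst gameplay cue on the player.
; It then calls Teleport.
; Enter and exit sounds sit behind a hard-coded false branch, so they never play.
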(event EventBeginPlay
  (|CallParentFunction) ; This node is disabled and will not be called.\nDrag off pins to build functionality.
)

(event EventPlayTeleportFX
  (params (PlayerLocation :type Vector :default "0, 0, 0") (Player :type Fort%20Player%20Pawn%20Object%20Reference))
  (bind _enterportal (Variables|VFX|GetEnterPortal))
  (Utilities|FlowControl|Delay 0.1)
  (Utilities|FlowControl|Sequence
    (:then_0
      (Effects|SpawnEmitteratLocation _enterportal (+ (Transformation|GetWorldLocation (Variables|/Script/Engine.Actor|GetRootComponent (Variables|Default|GetFrom%20Portal))) (Variables|Default|GetPortal%20Opening%20Offset)) (Math|Rotator|CombineRotators (Transformation|GetActorRotation (Variables|Default|GetFrom%20Portal)) (Variables|Default|GetRotation)))
      (Effects|SpawnEmitteratLocation _enterportal (+ (Transformation|GetWorldLocation (Variables|/Script/Engine.Actor|GetRootComponent (Variables|Default|GetTo%20Portal))) (Variables|Default|GetPortal%20Opening%20Offset)) (Math|Rotator|CombineRotators (Transformation|GetActorRotation (Variables|Default|GetTo%20Portal)) (Variables|Default|GetRotation)))
      (GameplayCue|ExecuteGameplayCue%28Burst%29Local Player (Variables|Default|GetTeleportGameplayCue) (Ability|GameplayCue|MakeGameplayCueParameters :NormalizedMagnitude 0.0 :RawMagnitude 0.0 :EffectContext "()" :AggregatedSourceTags "(GameplayTags=)" :AggregatedTargetTags "(GameplayTags=)" :Location "0,0,0" :Normal "0,0,0" :GameplayEffectLevel 1 :AbilityLevel 1 :bReplicateLocationWhenUsingMinimalRepProxy false))
      (if (not true)
        (Audio|PlaySoundatLocation (Variables|Sound|GetEnteredPortal) PlayerLocation)
        (bind _to_portal (Variables|Default|GetTo%20Portal)
          (:then
            (Audio|PlaySoundatLocation (Variables|Sound|GetExitedPortal) (+ (Transformation|GetWorldLocation (Variables|/Script/Engine.Actor|GetRootComponent _to_portal)) (Variables|Default|GetPortal%20Opening%20Offset))))
          (:else))))
    (:then_1
      (Default|CallTeleport))))

(fn ConstructionScript
  (|CallParentFunction))
